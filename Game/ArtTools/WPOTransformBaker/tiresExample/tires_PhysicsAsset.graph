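# Physics asset for the tire example in the WPOTransformBaker art tools folder, giving a skeletal tire group simple collision bodies.
# Bones tire1 to tire4 each carry one fat, near-centered capsule of almost identical size, while toptire has only a tiny capsule.
# Four constraints form a star, joining each tire directly to toptire with identical swing and twist limit angles.
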PhysicsAsset /Game/ArtTools/WPOTransformBaker/tiresExample/tires_PhysicsAsset
bodies {
 tire1 [{shapeType=Capsule, length=54.1308, radius=54.1183, center=<-0.7044 -0.01 0>, rotation=<rot 4.6181 -90 -0.0408>}]
 tire2 [{shapeType=Capsule, length=54.1308, radius=54.1183, center=<-0.4725 -0.01 0>, rotation=<rot 4.6181 -90 -0.0408>}]
 tire3 [{shapeType=Capsule, length=54.1679, radius=54.1308, center=<1.4797 0.0058 0>, rotation=<rot 85.3816 89.331 179.3332>}]
 tire4 [{shapeType=Capsule, length=54.1308, radius=54.1193, center=<0.2767 -0.5648 0.0001>, rotation=<rot 0.0409 -90 4.6181>}]
 toptire [{shapeType=Capsule, length=0.505, radius=0.505, rotation=<rot 0 -90 0>}]
}
constraints [
 {bone1Name=tire1, bone2Name=toptire, swing1LimitDegrees=45, swing2LimitDegrees=45, twistLimitDegrees=45}
 {bone1Name=tire2, bone2Name=toptire, swing1LimitDegrees=45, swing2LimitDegrees=45, twistLimitDegrees=45}
 {bone1Name=tire3, bone2Name=toptire, swing1LimitDegrees=45, swing2LimitDegrees=45, twistLimitDegrees=45}
 {bone1Name=tire4, bone2Name=toptire, swing1LimitDegrees=45, swing2LimitDegrees=45, twistLimitDegrees=45}
]
